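# Parent material for the AnglePatch Royal boss body, adding glowing damage cracks over the standard character shading.
# PreFX output gets metal roughness adjustment and car clear coat, then crack color, emissive and normal edits before PostFX.
# Pre-skinned-position sphere masks, noise-eroded for chest cracks and burst glow, grow by Damage_State, with three extra damage spots.
# Blueprint-set timestamps drive the transitions; a sine pulses crack glow, compensated for eye adaptation.
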Material /BRCosmetics/Characters/Player/Male/Medium/Bodies/M_MED_AnglePatch_Royal/Materials/BossMaterials_Updated/M_AnglePatch_Royal_Boss_PARENT
graph {
 expressions {
  MaterialExpressionAdd_10 {
   class=MaterialExpressionAdd
   inputs [
    A {from=MaterialExpressionTime_2}
    B {from=""}
   ]
   properties {constB=-1}
  }
  MaterialExpressionAdd_11 {
   class=MaterialExpressionAdd
   inputs [
    A {from=MaterialExpressionMultiply_26}
    B {from=""}
   ]
   properties {constB=0.5}
  }
  MaterialExpressionAdd_12 {
   class=MaterialExpressionAdd
   inputs [
    A {from=MaterialExpressionAdd_9}
    B {from=MaterialExpressionNamedRerouteUsage_10}
   ]
  }
  MaterialExpressionAdd_3 {
   class=MaterialExpressionAdd
   inputs [
    A {from=MaterialExpressionSphereMask_3}
    B {from=MaterialExpressionScalarParameter_8}
   ]
  }
  MaterialExpressionAdd_5 {
   class=MaterialExpressionAdd
   inputs [
    A {from=MaterialExpressionNamedRerouteUsage_18}
    B {from=MaterialExpressionNamedRerouteUsage_11}
   ]
  }
  MaterialExpressionAdd_6 {
   class=MaterialExpressionAdd
   inputs [
    A {from=MaterialExpressionTime_1}
    B {from=""}
   ]
   properties {constB=-0.8}
  }
  MaterialExpressionAdd_7 {
   class=MaterialExpressionAdd
   inputs [
    A {from=MaterialExpressionAdd_5}
    B {from=MaterialExpressionLinearInterpolate_10}
   ]
  }
  MaterialExpressionAdd_8 {
   class=MaterialExpressionAdd
   inputs [
    A {from=MaterialExpressionSphereMask_6}
    B {from=MaterialExpressionSphereMask_5}
   ]
  }
  MaterialExpressionAdd_9 {
   class=MaterialExpressionAdd
   inputs [
    A {from=MaterialExpressionAdd_8}
    B {from=MaterialExpressionSphereMask_7}
   ]
  }
  MaterialExpressionClearCoatNormalCustomOutput_0 {
   class=MaterialExpressionClearCoatNormalCustomOutput
   inputs [None {from=MaterialExpressionMaterialFunctionCall_9, output=Result}]
   properties {}
  }
  MaterialExpressionConstant_3 {class=MaterialExpressionConstant}
  MaterialExpressionConstant_4 {class=MaterialExpressionConstant, properties {r=13}}
  MaterialExpressionConstant_5 {class=MaterialExpressionConstant, properties {r=13}}
  MaterialExpressionConstant_8 {class=MaterialExpressionConstant, properties {r=1}}
  MaterialExpressionDivide_0 {
   class=MaterialExpressionDivide
   inputs [
    A {from=MaterialExpressionMultiply_7}
    B {from=MaterialExpressionEyeAdaptation_0, output=EyeAdaptation}
   ]
  }
  MaterialExpressionEyeAdaptation_0 {class=MaterialExpressionEyeAdaptation, inputs [], properties {}}
  MaterialExpressionGetMaterialAttributes_2 {
   class=MaterialExpressionGetMaterialAttributes
   inputs [None {from=MaterialExpressionMaterialFunctionCall_7, output=Result}]
   properties {attributeGetTypes [#1, #2]}
  }
  MaterialExpressionGetMaterialAttributes_4 {
   class=MaterialExpressionGetMaterialAttributes
   inputs [None {from=MaterialExpressionMaterialFunctionCall_8, output=Result}]
   properties {attributeGetTypes [#3, #4, #5]}
  }
  MaterialExpressionIf_1 {
   class=MaterialExpressionIf
   inputs [
    A {from=MaterialExpressionSaturate_1, output=""}
    B {from="", output=""}
    "A > B" {from=MaterialExpressionLinearInterpolate_2, output=""}
    "A == B" {from=MaterialExpressionVectorParameter_7, output=RGB}
    "A < B" {from=MaterialExpressionVectorParameter_7, output=RGB}
   ]
   properties {}
  }
  MaterialExpressionLinearInterpolate_0 {
   class=MaterialExpressionLinearInterpolate
   inputs [
    A {from=MaterialExpressionNamedRerouteUsage_17}
    B {from=""}
    Alpha {from=MaterialExpressionSaturate_0}
   ]
   properties {constB=0.05}
  }
  MaterialExpressionLinearInterpolate_1 {
   class=MaterialExpressionLinearInterpolate
   inputs [
    A {from=MaterialExpressionGetMaterialAttributes_2, output=Roughness}
    B {from=MaterialExpressionMultiply_1}
    Alpha {from=MaterialExpressionSaturate_2}
   ]
  }
  MaterialExpressionLinearInterpolate_10 {
   class=MaterialExpressionLinearInterpolate
   inputs [
    A {from=""}
    B {from=MaterialExpressionMultiply_25}
    Alpha {from=MaterialExpressionNamedRerouteUsage_12}
   ]
  }
  MaterialExpressionLinearInterpolate_11 {
   class=MaterialExpressionLinearInterpolate
   inputs [
    A {from=MaterialExpressionScalarParameter_2}
    B {from=MaterialExpressionScalarParameter_1}
    Alpha {from=MaterialExpressionAdd_11}
   ]
   properties {constA=0.4}
  }
  MaterialExpressionLinearInterpolate_2 {
   class=MaterialExpressionLinearInterpolate
   inputs [
    A {from=MaterialExpressionVectorParameter_0, output=RGB}
    B {from=MaterialExpressionVectorParameter_7, output=RGB}
    Alpha {from=MaterialExpressionSaturate_1}
   ]
  }
  MaterialExpressionLinearInterpolate_3 {
   class=MaterialExpressionLinearInterpolate
   inputs [
    A {from=""}
    B {from=MaterialExpressionSwitch_4}
    Alpha {from=MaterialExpressionNamedRerouteUsage_1}
   ]
   properties {constB=16}
  }
  MaterialExpressionLinearInterpolate_4 {
   class=MaterialExpressionLinearInterpolate
   inputs [
    A {from=""}
    B {from=MaterialExpressionScalarParameter_6}
    Alpha {from=MaterialExpressionNamedRerouteUsage_0}
   ]
  }
  MaterialExpressionLinearInterpolate_5 {
   class=MaterialExpressionLinearInterpolate
   inputs [
    A {from=MaterialExpressionScalarParameter_6}
    B {from=MaterialExpressionScalarParameter_12}
    Alpha {from=MaterialExpressionNamedRerouteUsage_2}
   ]
  }
  MaterialExpressionLinearInterpolate_6 {
   class=MaterialExpressionLinearInterpolate
   inputs [
    A {from=MaterialExpressionScalarParameter_12}
    B {from=MaterialExpressionScalarParameter_13}
    Alpha {from=MaterialExpressionNamedRerouteUsage_3}
   ]
  }
  MaterialExpressionLinearInterpolate_7 {
   class=MaterialExpressionLinearInterpolate
   inputs [
    A {from=MaterialExpressionScalarParameter_13}
    B {from=MaterialExpressionScalarParameter_14}
    Alpha {from=MaterialExpressionNamedRerouteUsage_4}
   ]
  }
  MaterialExpressionLinearInterpolate_9 {
   class=MaterialExpressionLinearInterpolate
   inputs [
    A {from=MaterialExpressionNamedRerouteUsage_16}
    B {from=MaterialExpressionTextureSample_6, output=RGB}
    Alpha {from=MaterialExpressionNamedRerouteUsage_13}
   ]
  }
  MaterialExpressionLocalPosition_1 {class=MaterialExpressionLocalPosition, inputs [], properties {localOrigin=InstancePreSkinning}}
  MaterialExpressionMaterialFunctionCall_6 {
   class=MaterialExpressionMaterialFunctionCall
   inputs [
    Material {from=MaterialExpressionSetMaterialAttributes_4}
    M {from=MaterialExpressionMaterialFunctionCall_7, output=M}
    Pre Skinned Local Position {from=MaterialExpressionMaterialFunctionCall_7, output=M}
    Distance Blend {from=MaterialExpressionMaterialFunctionCall_7, output=M}
    Fresnel {from=MaterialExpressionMaterialFunctionCall_7, output=M}
   ]
   properties {materialFunction=&MF_CharacterParent_2023_PostFX}
  }
  MaterialExpressionMaterialFunctionCall_7 {
   class=MaterialExpressionMaterialFunctionCall
   inputs [
    Overide Texture Set UVs {from=""}
    Override Tech Art Mask UVs {from=""}
    Override Basic Emissive UVs {from=""}
   ]
   properties {materialFunction=&MF_CharacterParent_2023_PreFX}
  }
  MaterialExpressionMaterialFunctionCall_8 {
   class=MaterialExpressionMaterialFunctionCall
   inputs [
    Mask {from=MaterialExpressionMaterialFunctionCall_7, output=Tech Art Mask}
    Fresnel {from=MaterialExpressionMaterialFunctionCall_7, output=Tech Art Mask}
    In {from=MaterialExpressionSetMaterialAttributes_2}
   ]
   properties {materialFunction=/Game/Global/MaterialFunctions/TechArt/MF_Advanced_CarClearCoat}
  }
  MaterialExpressionMaterialFunctionCall_9 {
   class=MaterialExpressionMaterialFunctionCall
   inputs [
    BaseNormal {from=MaterialExpressionMaterialFunctionCall_8, output=NormalWithOrangePeel}
    AdditionalNormal {from=MaterialExpressionMaterialFunctionCall_8, output=NormalWithOrangePeel}
   ]
   properties {materialFunction=&BlendAngleCorrectedNormals}
  }
  MaterialExpressionMultiply_0 {
   class=MaterialExpressionMultiply
   inputs [
    A {from=MaterialExpressionTextureSample_5, output=RGB}
    B {from=MaterialExpressionNamedRerouteUsage_20}
   ]
  }
  MaterialExpressionMultiply_1 {
   class=MaterialExpressionMultiply
   inputs [
    A {from=MaterialExpressionGetMaterialAttributes_2, output=Roughness}
    B {from=MaterialExpressionScalarParameter_5}
   ]
  }
  MaterialExpressionMultiply_10 {
   class=MaterialExpressionMultiply
   inputs [
    A {from=MaterialExpressionSphereMask_4}
    B {from=MaterialExpressionOneMinus_0}
   ]
  }
  MaterialExpressionMultiply_11 {
   class=MaterialExpressionMultiply
   inputs [
    A {from=MaterialExpressionSubtract_4}
    B {from=""}
   ]
   properties {constB=2}
  }
  MaterialExpressionMultiply_12 {
   class=MaterialExpressionMultiply
   inputs [
    A {from=MaterialExpressionSaturate_8}
    B {from=MaterialExpressionVectorParameter_1, output=RGB}
   ]
  }
  MaterialExpressionMultiply_13 {
   class=MaterialExpressionMultiply
   inputs [
    A {from=MaterialExpressionMultiply_12}
    B {from=""}
   ]
  }
  MaterialExpressionMultiply_15 {
   class=MaterialExpressionMultiply
   inputs [
    A {from=MaterialExpressionScalarParameter_14}
    B {from=MaterialExpressionConstant_8}
   ]
   properties {constB=0.7}
  }
  MaterialExpressionMultiply_16 {
   class=MaterialExpressionMultiply
   inputs [
    A {from=MaterialExpressionScalarParameter_13}
    B {from=MaterialExpressionConstant_8}
   ]
   properties {constB=0.7}
  }
  MaterialExpressionMultiply_17 {
   class=MaterialExpressionMultiply
   inputs [
    A {from=MaterialExpressionScalarParameter_12}
    B {from=MaterialExpressionConstant_8}
   ]
   properties {constB=0.7}
  }
  MaterialExpressionMultiply_19 {
   class=MaterialExpressionMultiply
   inputs [
    A {from=MaterialExpressionSubtract_5}
    B {from=""}
   ]
  }
  MaterialExpressionMultiply_2 {
   class=MaterialExpressionMultiply
   inputs [
    A {from=MaterialExpressionMultiply_0}
    B {from=""}
   ]
   properties {constB=100}
  }
  MaterialExpressionMultiply_20 {
   class=MaterialExpressionMultiply
   inputs [
    A {from=MaterialExpressionScalarParameter_18}
    B {from=MaterialExpressionSaturate_12}
   ]
  }
  MaterialExpressionMultiply_21 {
   class=MaterialExpressionMultiply
   inputs [
    A {from=MaterialExpressionSubtract_3}
    B {from=""}
   ]
   properties {constB=5}
  }
  MaterialExpressionMultiply_22 {
   class=MaterialExpressionMultiply
   inputs [
    A {from=MaterialExpressionSubtract_6}
    B {from=""}
   ]
  }
  MaterialExpressionMultiply_23 {
   class=MaterialExpressionMultiply
   inputs [
    A {from=MaterialExpressionScalarParameter_16}
    B {from=MaterialExpressionSaturate_13}
   ]
  }
  MaterialExpressionMultiply_24 {
   class=MaterialExpressionMultiply
   inputs [
    A {from=MaterialExpressionScalarParameter_17}
    B {from=MaterialExpressionSaturate_14}
   ]
  }
  MaterialExpressionMultiply_25 {
   class=MaterialExpressionMultiply
   inputs [
    A {from=MaterialExpressionDivide_0}
    B {from=MaterialExpressionLinearInterpolate_11}
   ]
  }
  MaterialExpressionMultiply_26 {
   class=MaterialExpressionMultiply
   inputs [
    A {from=MaterialExpressionSine_0}
    B {from=""}
   ]
   properties {constB=0.5}
  }
  MaterialExpressionMultiply_27 {
   class=MaterialExpressionMultiply
   inputs [
    A {from=MaterialExpressionTime_4}
    B {from=""}
   ]
   properties {constB=0.37}
  }
  MaterialExpressionMultiply_6 {
   class=MaterialExpressionMultiply
   inputs [
    A {from=MaterialExpressionTextureCoordinate_1}
    B {from=""}
   ]
   properties {constB=5}
  }
  MaterialExpressionMultiply_7 {
   class=MaterialExpressionMultiply
   inputs [
    A {from=MaterialExpressionTextureSample_5, output=RGB}
    B {from=MaterialExpressionLinearInterpolate_2}
   ]
  }
  MaterialExpressionMultiply_8 {
   class=MaterialExpressionMultiply
   inputs [
    A {from=MaterialExpressionSubtract_1}
    B {from=""}
   ]
   properties {constB=0.5}
  }
  MaterialExpressionNamedRerouteDeclaration_0 {
   class=MaterialExpressionNamedRerouteDeclaration
   inputs [None {from=MaterialExpressionSaturate_7}]
   properties {name=AnimatedTime_Cracks, nodeColor=<rgba 0.4766 1 0 1>}
  }
  MaterialExpressionNamedRerouteDeclaration_1 {
   class=MaterialExpressionNamedRerouteDeclaration
   inputs [None {from=MaterialExpressionSaturate_6}]
   properties {name=AnimatedTime _Glow, nodeColor=<rgba 1 0 0.7969 1>}
  }
  MaterialExpressionNamedRerouteDeclaration_10 {
   class=MaterialExpressionNamedRerouteDeclaration
   inputs [None {from=MaterialExpressionGetMaterialAttributes_4, output=Normal}]
   properties {name=OriginalNormal, nodeColor=<rgba 0 0.9531 1 1>}
  }
  MaterialExpressionNamedRerouteDeclaration_11 {
   class=MaterialExpressionNamedRerouteDeclaration
   inputs [None {from=MaterialExpressionGetMaterialAttributes_4, output=Base Color}]
   properties {name=OriginalBaseColor, nodeColor=<rgba 1 0.7266 0 1>}
  }
  MaterialExpressionNamedRerouteDeclaration_12 {
   class=MaterialExpressionNamedRerouteDeclaration
   inputs [None {from=MaterialExpressionGetMaterialAttributes_4, output=Emissive Color}]
   properties {name=OriginalEmissiveColor, nodeColor=<rgba 0.4062 0 1 1>}
  }
  MaterialExpressionNamedRerouteDeclaration_13 {
   class=MaterialExpressionNamedRerouteDeclaration
   inputs [None {from=MaterialExpressionLinearInterpolate_0}]
   properties {name=BaseColorOUT, nodeColor=<rgba 0 1 0.0859 1>}
  }
  MaterialExpressionNamedRerouteDeclaration_2 {
   class=MaterialExpressionNamedRerouteDeclaration
   inputs [None {from=MaterialExpressionSaturate_5}]
   properties {name=ChestMask, nodeColor=<rgba 0.5234 1 0 1>}
  }
  MaterialExpressionNamedRerouteDeclaration_3 {
   class=MaterialExpressionNamedRerouteDeclaration
   inputs [None {from=MaterialExpressionVectorParameter_3, output=RGB}]
   properties {name=CracksMask_Position, nodeColor=<rgba 0 1 0.5547 1>}
  }
  MaterialExpressionNamedRerouteDeclaration_4 {
   class=MaterialExpressionNamedRerouteDeclaration
   inputs [None {from=MaterialExpressionShaderStageSwitch_1}]
   properties {name=PreskinnedPosition, nodeColor=<rgba 1 0.2344 0 1>}
  }
  MaterialExpressionNamedRerouteDeclaration_5 {
   class=MaterialExpressionNamedRerouteDeclaration
   inputs [None {from=MaterialExpressionSaturate_9}]
   properties {name=AnimatedTime_Scar, nodeColor=<rgba 0.4766 1 0 1>}
  }
  MaterialExpressionNamedRerouteDeclaration_6 {
   class=MaterialExpressionNamedRerouteDeclaration
   inputs [None {from=MaterialExpressionMultiply_13}]
   properties {name=SphereGlow, nodeColor=<rgba 1 0.3281 0 1>}
  }
  MaterialExpressionNamedRerouteDeclaration_7 {
   class=MaterialExpressionNamedRerouteDeclaration
   inputs [None {from=MaterialExpressionSaturate_11}]
   properties {name=DamageSphereMasks, nodeColor=<rgba 0.0078 0 1 1>}
  }
  MaterialExpressionNamedRerouteDeclaration_8 {
   class=MaterialExpressionNamedRerouteDeclaration
   inputs [None {from=MaterialExpressionAdd_7}]
   properties {name=EmissiveColorOut, nodeColor=<rgba 0.3125 1 0 1>}
  }
  MaterialExpressionNamedRerouteDeclaration_9 {
   class=MaterialExpressionNamedRerouteDeclaration
   inputs [None {from=MaterialExpressionLinearInterpolate_9}]
   properties {name=NormalOut, nodeColor=<rgba 1 0 0.6328 1>}
  }
  MaterialExpressionNamedRerouteUsage_0 {class=MaterialExpressionNamedRerouteUsage}
  MaterialExpressionNamedRerouteUsage_1 {class=MaterialExpressionNamedRerouteUsage}
  MaterialExpressionNamedRerouteUsage_10 {class=MaterialExpressionNamedRerouteUsage}
  MaterialExpressionNamedRerouteUsage_11 {class=MaterialExpressionNamedRerouteUsage}
  MaterialExpressionNamedRerouteUsage_12 {class=MaterialExpressionNamedRerouteUsage}
  MaterialExpressionNamedRerouteUsage_13 {class=MaterialExpressionNamedRerouteUsage}
  MaterialExpressionNamedRerouteUsage_14 {class=MaterialExpressionNamedRerouteUsage}
  MaterialExpressionNamedRerouteUsage_15 {class=MaterialExpressionNamedRerouteUsage}
  MaterialExpressionNamedRerouteUsage_16 {class=MaterialExpressionNamedRerouteUsage}
  MaterialExpressionNamedRerouteUsage_17 {class=MaterialExpressionNamedRerouteUsage}
  MaterialExpressionNamedRerouteUsage_18 {class=MaterialExpressionNamedRerouteUsage}
  MaterialExpressionNamedRerouteUsage_19 {class=MaterialExpressionNamedRerouteUsage}
  MaterialExpressionNamedRerouteUsage_2 {class=MaterialExpressionNamedRerouteUsage}
  MaterialExpressionNamedRerouteUsage_20 {class=MaterialExpressionNamedRerouteUsage}
  MaterialExpressionNamedRerouteUsage_3 {class=MaterialExpressionNamedRerouteUsage}
  MaterialExpressionNamedRerouteUsage_4 {class=MaterialExpressionNamedRerouteUsage}
  MaterialExpressionNamedRerouteUsage_5 {class=MaterialExpressionNamedRerouteUsage}
  MaterialExpressionNamedRerouteUsage_6 {class=MaterialExpressionNamedRerouteUsage}
  MaterialExpressionNamedRerouteUsage_7 {class=MaterialExpressionNamedRerouteUsage}
  MaterialExpressionNamedRerouteUsage_8 {class=MaterialExpressionNamedRerouteUsage}
  MaterialExpressionOneMinus_0 {class=MaterialExpressionOneMinus, inputs [None {from=MaterialExpressionNamedRerouteUsage_1}]}
  MaterialExpressionPower_1 {
   class=MaterialExpressionPower
   inputs [
    Base {from=MaterialExpressionGetMaterialAttributes_2, output=Metallic}
    Exp {from=""}
   ]
  }
  MaterialExpressionReroute_0 {class=MaterialExpressionReroute, inputs [None {from=MaterialExpressionSaturate_7}]}
  MaterialExpressionReroute_2 {class=MaterialExpressionReroute, inputs [None {from=MaterialExpressionSaturate_3}]}
  MaterialExpressionReroute_4 {class=MaterialExpressionReroute, inputs [None {from=MaterialExpressionVectorParameter_3, output=RGB}]}
  MaterialExpressionReroute_5 {class=MaterialExpressionReroute, inputs [None {from=MaterialExpressionShaderStageSwitch_1}]}
  MaterialExpressionReroute_6 {class=MaterialExpressionReroute, inputs [None {from=MaterialExpressionMultiply_13}]}
  MaterialExpressionSaturate_0 {class=MaterialExpressionSaturate, inputs [None {from=MaterialExpressionMultiply_2}]}
  MaterialExpressionSaturate_1 {class=MaterialExpressionSaturate, inputs [None {from=MaterialExpressionSphereMask_0}]}
  MaterialExpressionSaturate_11 {class=MaterialExpressionSaturate, inputs [None {from=MaterialExpressionAdd_12}]}
  MaterialExpressionSaturate_12 {class=MaterialExpressionSaturate, inputs [None {from=MaterialExpressionSubtract_5}]}
  MaterialExpressionSaturate_13 {class=MaterialExpressionSaturate, inputs [None {from=MaterialExpressionSubtract_6}]}
  MaterialExpressionSaturate_14 {class=MaterialExpressionSaturate, inputs [None {from=MaterialExpressionSubtract_7}]}
  MaterialExpressionSaturate_2 {class=MaterialExpressionSaturate, inputs [None {from=MaterialExpressionPower_1}]}
  MaterialExpressionSaturate_3 {class=MaterialExpressionSaturate, inputs [None {from=MaterialExpressionAdd_3}]}
  MaterialExpressionSaturate_5 {class=MaterialExpressionSaturate, inputs [None {from=MaterialExpressionSubtract_0}]}
  MaterialExpressionSaturate_6 {class=MaterialExpressionSaturate, inputs [None {from=MaterialExpressionMultiply_8}]}
  MaterialExpressionSaturate_7 {class=MaterialExpressionSaturate, inputs [None {from=MaterialExpressionMultiply_11}]}
  MaterialExpressionSaturate_8 {class=MaterialExpressionSaturate, inputs [None {from=MaterialExpressionSubtract_2}]}
  MaterialExpressionSaturate_9 {class=MaterialExpressionSaturate, inputs [None {from=MaterialExpressionMultiply_21}]}
  MaterialExpressionScalarParameter_0 {class=MaterialExpressionScalarParameter, properties {defaultValue=10000000272564224, parameterName=BP_GetTime_Damage03}}
  MaterialExpressionScalarParameter_1 {
   class=MaterialExpressionScalarParameter
   properties {defaultValue=80, group=CracksMask, parameterName=Cracks_GlowIntensity_Max, sortPriority=60}
  }
  MaterialExpressionScalarParameter_10 {class=MaterialExpressionScalarParameter, properties {defaultValue=999999986991104, parameterName=BP_GetTime}}
  MaterialExpressionScalarParameter_11 {class=MaterialExpressionScalarParameter, properties {group=CracksMask, parameterName=Damage_State}}
  MaterialExpressionScalarParameter_12 {class=MaterialExpressionScalarParameter, properties {defaultValue=20, group=CracksMask, parameterName=CracksMask01_Radius02}}
  MaterialExpressionScalarParameter_13 {class=MaterialExpressionScalarParameter, properties {defaultValue=25, group=CracksMask, parameterName=CracksMask01_Radius03}}
  MaterialExpressionScalarParameter_14 {class=MaterialExpressionScalarParameter, properties {defaultValue=30, group=CracksMask, parameterName=CracksMask01_Radius04}}
  MaterialExpressionScalarParameter_16 {class=MaterialExpressionScalarParameter, properties {defaultValue=1, group=CracksMask, parameterName=Damage02_Radius, sortPriority=100}}
  MaterialExpressionScalarParameter_17 {class=MaterialExpressionScalarParameter, properties {defaultValue=1, group=CracksMask, parameterName=Damage01_Radius, sortPriority=100}}
  MaterialExpressionScalarParameter_18 {class=MaterialExpressionScalarParameter, properties {defaultValue=1, group=CracksMask, parameterName=Damage03_Radius, sortPriority=100}}
  MaterialExpressionScalarParameter_19 {class=MaterialExpressionScalarParameter, properties {defaultValue=999999986991104, parameterName=BP_GetTime}}
  MaterialExpressionScalarParameter_2 {
   class=MaterialExpressionScalarParameter
   properties {defaultValue=80, group=CracksMask, parameterName=Cracks_GlowIntensity_Min, sortPriority=60}
  }
  MaterialExpressionScalarParameter_20 {class=MaterialExpressionScalarParameter, properties {defaultValue=10000000272564224, parameterName=BP_GetTime_Damage02}}
  MaterialExpressionScalarParameter_21 {class=MaterialExpressionScalarParameter, properties {defaultValue=10000000272564224, parameterName=BP_GetTime_Damage01}}
  MaterialExpressionScalarParameter_3 {class=MaterialExpressionScalarParameter, properties {defaultValue=1, group=CracksMask, parameterName=ChestGradient_Radius, sortPriority=60}}
  MaterialExpressionScalarParameter_4 {
   class=MaterialExpressionScalarParameter
   properties {defaultValue=0.5, group=CracksMask, parameterName=ChestGradient_Hardness, sortPriority=60}
  }
  MaterialExpressionScalarParameter_5 {class=MaterialExpressionScalarParameter, properties {defaultValue=1, parameterName=Metal Roughness Multiplier}}
  MaterialExpressionScalarParameter_6 {class=MaterialExpressionScalarParameter, properties {defaultValue=10, group=CracksMask, parameterName=CracksMask01_Radius01}}
  MaterialExpressionScalarParameter_7 {class=MaterialExpressionScalarParameter, properties {defaultValue=100, group=CracksMask, parameterName=CracksMask_Hardness}}
  MaterialExpressionScalarParameter_8 {class=MaterialExpressionScalarParameter, properties {group=CracksMask, parameterName=CracksMask_Min}}
  MaterialExpressionScalarParameter_9 {class=MaterialExpressionScalarParameter, properties {defaultValue=999999986991104, parameterName=BP_GetTime}}
  MaterialExpressionSetMaterialAttributes_2 {
   class=MaterialExpressionSetMaterialAttributes
   inputs [
    MaterialAttributes {from=MaterialExpressionGetMaterialAttributes_2, output=MaterialAttributes}
    Roughness {from=MaterialExpressionLinearInterpolate_1, output=""}
   ]
   properties {attributeSetTypes [#1]}
  }
  MaterialExpressionSetMaterialAttributes_4 {
   class=MaterialExpressionSetMaterialAttributes
   inputs [
    MaterialAttributes {from=MaterialExpressionGetMaterialAttributes_4, output=MaterialAttributes}
    Emissive Color {from=MaterialExpressionNamedRerouteUsage_14, output=""}
    Normal {from=MaterialExpressionNamedRerouteUsage_15, output=""}
    Base Color {from=MaterialExpressionNamedRerouteUsage_19, output=""}
   ]
   properties {attributeSetTypes [#3, #4, #5]}
  }
  MaterialExpressionShaderStageSwitch_1 {
   class=MaterialExpressionShaderStageSwitch
   inputs [
    PixelShader {from=MaterialExpressionVertexInterpolator_1, output=PS}
    VertexShader {from=MaterialExpressionLocalPosition_1, output=XYZ}
   ]
   properties {}
  }
  MaterialExpressionSine_0 {class=MaterialExpressionSine, inputs [None {from=MaterialExpressionMultiply_27}]}
  MaterialExpressionSphereMask_0 {
   class=MaterialExpressionSphereMask
   inputs [
    A {from=MaterialExpressionNamedRerouteUsage_5, output=""}
    B {from=MaterialExpressionVectorParameter_2, output=RGB}
    Radius {from=MaterialExpressionScalarParameter_3, output=""}
    Hardness {from=MaterialExpressionScalarParameter_4, output=""}
   ]
   properties {}
  }
  MaterialExpressionSphereMask_3 {
   class=MaterialExpressionSphereMask
   inputs [
    A {from=MaterialExpressionNamedRerouteDeclaration_4, output=""}
    B {from=MaterialExpressionNamedRerouteDeclaration_3, output=""}
    Radius {from=MaterialExpressionSwitch_2, output=""}
    Hardness {from=MaterialExpressionScalarParameter_7, output=""}
   ]
   properties {}
  }
  MaterialExpressionSphereMask_4 {
   class=MaterialExpressionSphereMask
   inputs [
    A {from=MaterialExpressionNamedRerouteDeclaration_4, output=""}
    B {from=MaterialExpressionNamedRerouteDeclaration_3, output=""}
    Radius {from=MaterialExpressionLinearInterpolate_3, output=""}
    Hardness {from="", output=""}
   ]
   properties {}
  }
  MaterialExpressionSphereMask_5 {
   class=MaterialExpressionSphereMask
   inputs [
    A {from=MaterialExpressionVectorParameter_4, output=RGB}
    B {from=MaterialExpressionNamedRerouteUsage_6, output=""}
    Radius {from=MaterialExpressionMultiply_23, output=""}
    Hardness {from="", output=""}
   ]
   properties {}
  }
  MaterialExpressionSphereMask_6 {
   class=MaterialExpressionSphereMask
   inputs [
    A {from=MaterialExpressionVectorParameter_5, output=RGB}
    B {from=MaterialExpressionNamedRerouteUsage_7, output=""}
    Radius {from=MaterialExpressionMultiply_24, output=""}
    Hardness {from="", output=""}
   ]
   properties {}
  }
  MaterialExpressionSphereMask_7 {
   class=MaterialExpressionSphereMask
   inputs [
    A {from=MaterialExpressionVectorParameter_6, output=RGB}
    B {from=MaterialExpressionNamedRerouteUsage_8, output=""}
    Radius {from=MaterialExpressionMultiply_20, output=""}
    Hardness {from="", output=""}
   ]
   properties {}
  }
  MaterialExpressionSubtract_0 {
   class=MaterialExpressionSubtract
   inputs [
    A {from=MaterialExpressionSaturate_3}
    B {from=MaterialExpressionTextureSample_2, output=G}
   ]
  }
  MaterialExpressionSubtract_1 {
   class=MaterialExpressionSubtract
   inputs [
    A {from=MaterialExpressionTime_0}
    B {from=MaterialExpressionScalarParameter_9}
   ]
  }
  MaterialExpressionSubtract_2 {
   class=MaterialExpressionSubtract
   inputs [
    A {from=MaterialExpressionMultiply_10}
    B {from=MaterialExpressionTextureSample_2, output=G}
   ]
  }
  MaterialExpressionSubtract_3 {
   class=MaterialExpressionSubtract
   inputs [
    A {from=MaterialExpressionAdd_10}
    B {from=MaterialExpressionScalarParameter_19}
   ]
  }
  MaterialExpressionSubtract_4 {
   class=MaterialExpressionSubtract
   inputs [
    A {from=MaterialExpressionAdd_6}
    B {from=MaterialExpressionScalarParameter_10}
   ]
  }
  MaterialExpressionSubtract_5 {
   class=MaterialExpressionSubtract
   inputs [
    A {from=MaterialExpressionTime_3}
    B {from=MaterialExpressionScalarParameter_0}
   ]
  }
  MaterialExpressionSubtract_6 {
   class=MaterialExpressionSubtract
   inputs [
    A {from=MaterialExpressionTime_3}
    B {from=MaterialExpressionScalarParameter_20}
   ]
  }
  MaterialExpressionSubtract_7 {
   class=MaterialExpressionSubtract
   inputs [
    A {from=MaterialExpressionTime_3}
    B {from=MaterialExpressionScalarParameter_21}
   ]
  }
  MaterialExpressionSwitch_2 {
   class=MaterialExpressionSwitch
   inputs [
    SwitchValue {from=MaterialExpressionScalarParameter_11, output=""}
    Default {from=MaterialExpressionConstant_3, output=""}
    None {from=MaterialExpressionLinearInterpolate_4, output=""}
    None {from=MaterialExpressionLinearInterpolate_4, output=""}
    None {from=MaterialExpressionLinearInterpolate_5, output=""}
    None {from=MaterialExpressionLinearInterpolate_6, output=""}
    None {from=MaterialExpressionLinearInterpolate_7, output=""}
   ]
   properties {inputs [
     {
      inputName=None
      input {
       expression=&self:MaterialExpressionLinearInterpolate_4
       inputName=None
       mask=0
       maskA=0
       maskB=0
       maskG=0
       maskR=0
       outputIndex=0
      }
     }
     {
      inputName=None
      input {
       expression=&self:MaterialExpressionLinearInterpolate_4
       inputName=None
       mask=0
       maskA=0
       maskB=0
       maskG=0
       maskR=0
       outputIndex=0
      }
     }
     {
      inputName=None
      input {
       expression=&self:MaterialExpressionLinearInterpolate_5
       inputName=None
       mask=0
       maskA=0
       maskB=0
       maskG=0
       maskR=0
       outputIndex=0
      }
     }
     {
      inputName=None
      input {
       expression=&self:MaterialExpressionLinearInterpolate_6
       inputName=None
       mask=0
       maskA=0
       maskB=0
       maskG=0
       maskR=0
       outputIndex=0
      }
     }
     {
      inputName=None
      input {
       expression=&self:MaterialExpressionLinearInterpolate_7
       inputName=None
       mask=0
       maskA=0
       maskB=0
       maskG=0
       maskR=0
       outputIndex=0
      }
     }
    ]}
  }
  MaterialExpressionSwitch_4 {
   class=MaterialExpressionSwitch
   inputs [
    SwitchValue {from=MaterialExpressionScalarParameter_11, output=""}
    Default {from=MaterialExpressionConstant_4, output=""}
    None {from=MaterialExpressionConstant_4, output=""}
    None {from=MaterialExpressionConstant_5, output=""}
    None {from=MaterialExpressionMultiply_17, output=""}
    None {from=MaterialExpressionMultiply_16, output=""}
    None {from=MaterialExpressionMultiply_15, output=""}
   ]
   properties {
    constDefault=13
    inputs [
     {
      inputName=None
      input {
       expression=&self:MaterialExpressionConstant_4
       inputName=None
       mask=0
       maskA=0
       maskB=0
       maskG=0
       maskR=0
       outputIndex=0
      }
     }
     {
      inputName=None
      input {
       expression=&self:MaterialExpressionConstant_5
       inputName=None
       mask=0
       maskA=0
       maskB=0
       maskG=0
       maskR=0
       outputIndex=0
      }
     }
     {
      inputName=None
      input {
       expression=&self:MaterialExpressionMultiply_17
       inputName=None
       mask=0
       maskA=0
       maskB=0
       maskG=0
       maskR=0
       outputIndex=0
      }
     }
     {
      inputName=None
      input {
       expression=&self:MaterialExpressionMultiply_16
       inputName=None
       mask=0
       maskA=0
       maskB=0
       maskG=0
       maskR=0
       outputIndex=0
      }
     }
     {
      inputName=None
      input {
       expression=&self:MaterialExpressionMultiply_15
       inputName=None
       mask=0
       maskA=0
       maskB=0
       maskG=0
       maskR=0
       outputIndex=0
      }
     }
    ]
   }
  }
  MaterialExpressionTextureCoordinate_1 {class=MaterialExpressionTextureCoordinate}
  MaterialExpressionTextureSample_2 {
   class=MaterialExpressionTextureSample
   inputs [
    UVs {from=MaterialExpressionMultiply_6}
    Tex {from=""}
    Apply View MipBias {from=""}
   ]
   properties {samplerSource=SSM_Wrap_WorldGroupSettings, texture=/Game/Characters/Player/Male/Medium/Bodies/M_MED_Viper/Textures/Glitch/T_SoftNoise}
  }
  MaterialExpressionTextureSample_5 {
   class=MaterialExpressionTextureSample
   inputs [
    UVs {from=""}
    Tex {from=""}
    Apply View MipBias {from=""}
   ]
   properties {
    samplerSource=SSM_Wrap_WorldGroupSettings
    texture=/BRCosmetics/Characters/Player/Male/Medium/Bodies/M_MED_AnglePatch_Royal/Textures/T_AnglePatch_Royal_Body_DamageCracks
   }
  }
  MaterialExpressionTextureSample_6 {
   class=MaterialExpressionTextureSample
   inputs [
    UVs {from=""}
    Tex {from=""}
    Apply View MipBias {from=""}
   ]
   properties {
    samplerSource=SSM_Wrap_WorldGroupSettings
    samplerType=SAMPLERTYPE_Normal
    texture=/BRCosmetics/Characters/Player/Male/Medium/Bodies/M_MED_AnglePatch_Royal/Textures/T_AnglePatch_Royal_Body_DamageCracks_N
   }
  }
  MaterialExpressionTime_0 {class=MaterialExpressionTime}
  MaterialExpressionTime_1 {class=MaterialExpressionTime}
  MaterialExpressionTime_2 {class=MaterialExpressionTime}
  MaterialExpressionTime_3 {class=MaterialExpressionTime}
  MaterialExpressionTime_4 {class=MaterialExpressionTime}
  MaterialExpressionVectorParameter_0 {
   class=MaterialExpressionVectorParameter
   properties {group=CracksMask, parameterName=Cracks_GlowColor01, sortPriority=60, defaultValue=<rgba 0.724 0.0298 0 1>}
  }
  MaterialExpressionVectorParameter_1 {
   class=MaterialExpressionVectorParameter
   properties {group=CracksMask, parameterName=Burst_GlowColor, sortPriority=60, defaultValue=<rgba 1 0.3389 0 1>}
  }
  MaterialExpressionVectorParameter_2 {class=MaterialExpressionVectorParameter, properties {group=CracksMask, parameterName=ChestGradient_Center, sortPriority=60}}
  MaterialExpressionVectorParameter_3 {class=MaterialExpressionVectorParameter, properties {group=CracksMask, parameterName=CracksMask_Pos01, defaultValue=<rgba 0 0 0 1>}}
  MaterialExpressionVectorParameter_4 {
   class=MaterialExpressionVectorParameter
   properties {group=CracksMask, parameterName=Damage02_Pos, sortPriority=100, defaultValue=<rgba 0 0 0 1>}
  }
  MaterialExpressionVectorParameter_5 {
   class=MaterialExpressionVectorParameter
   properties {group=CracksMask, parameterName=Damage01_Pos, sortPriority=100, defaultValue=<rgba 0 0 0 1>}
  }
  MaterialExpressionVectorParameter_6 {
   class=MaterialExpressionVectorParameter
   properties {group=CracksMask, parameterName=Damage03_Pos, sortPriority=100, defaultValue=<rgba 0 0 0 1>}
  }
  MaterialExpressionVectorParameter_7 {
   class=MaterialExpressionVectorParameter
   properties {group=CracksMask, parameterName=Cracks_GlowColor02, sortPriority=60, defaultValue=<rgba 0.724 0.0298 0 1>}
  }
  MaterialExpressionVertexInterpolator_1 {class=MaterialExpressionVertexInterpolator, inputs [VS {from=MaterialExpressionLocalPosition_1, output=XYZ}], properties {}}
 }
 properties {
  bUseMaterialAttributes=true
  bUsedWithMeshDeformer=true
  bUsedWithSkeletalMesh=true
  pixelDepthOffsetMode=PDOM_Legacy
  shadingModel=MSM_FromMaterialExpression
  usedShadingModels=Default Lit | Subsurface | Clear Coat | Subsurface Profile | Cloth
 }
}
