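# Parent material for the head of the boss variant of the AnglePatch Royal character body, built on material attributes for skeletal meshes and mesh deformers.
# It is a thin wrapper around the shared 2023 character parent functions, with a PreFX stage feeding a PostFX stage.
# PostFX inputs include material, pre-skinned local position, distance blend and fresnel.
# The shading model comes from the graph, spanning lit, subsurface, clear coat and cloth types.
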Material /BRCosmetics/Characters/Player/Male/Medium/Bodies/M_MED_AnglePatch_Royal/Materials/BossMaterials_Updated/M_AnglePatch_Royal_Boss_PARENT_Head
graph {
 properties {
  bUseMaterialAttributes=true
  bUsedWithMeshDeformer=true
  bUsedWithSkeletalMesh=true
  pixelDepthOffsetMode=PDOM_Legacy
  shadingModel=MSM_FromMaterialExpression
  usedShadingModels=Default Lit | Subsurface | Clear Coat | Subsurface Profile | Cloth
 }
 nodes [
  n1 = MaterialFunctionCall() {materialFunction=&MF_CharacterParent_2023_PreFX}
  MaterialFunctionCall(Material=@n1.Result, M=@n1.Result, Pre Skinned Local Position=@n1.Result, Distance Blend=@n1.Result, Fresnel=@n1.Result) {materialFunction=&MF_CharacterParent_2023_PostFX}
 ]
}
